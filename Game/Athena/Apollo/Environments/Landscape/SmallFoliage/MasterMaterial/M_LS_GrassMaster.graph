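# Master material for Apollo landscape small foliage: masked, two-sided foliage shading for instanced and Nanite meshes.
# It chains swamp grass, heather, basic grass, rock, leaf and jungle plant functions into snow progression.
# Base color is then brightened and desaturated before the Curie landscape function.
# A "Jungle Test" switch adds distance-field-AO dirt tinting and camera-depth roughness/specular adjustment, which mobile bypasses.
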Material /Game/Athena/Apollo/Environments/Landscape/SmallFoliage/MasterMaterial/M_LS_GrassMaster
graph {
 properties {
  bTangentSpaceNormal=false
  bUseMaterialAttributes=true
  bUsedWithInstancedStaticMeshes=true
  bUsedWithNanite=true
  blendMode=BLEND_Masked
  pixelDepthOffsetMode=PDOM_Legacy
  previewMesh=/Engine/Transient.MaterialEditorMeshComponent_41
  shadingModel=MSM_TwoSidedFoliage
  twoSided=true
  usedShadingModels=Two Sided Foliage
 }
 nodes [
  Constant()
  n1 = StaticBoolParameter() {defaultValue=true, desc="Disconnect this from \"IsStaticSnow\" when using dynamic snow", parameterName=Snow-IsForGrassMeshToPushDown}
  n2 = MaterialFunctionCall!(In=MaterialFunctionCall!(In=SetMaterialAttributes!(MaterialAttributes)) {materialFunction=/Game/Athena/Apollo/Environments/Landscape/SmallFoliage/MasterMaterial/MaterialFunctions/MF_SwampGrass}.Result) {materialFunction=/Game/Athena/Apollo/Environments/Landscape/SmallFoliage/MasterMaterial/MaterialFunctions/MF_Heather}
  n3 = MaterialFunctionCall!(In=MaterialFunctionCall!(In=MaterialFunctionCall!(In=@n2.Result) {materialFunction=/Game/Athena/Apollo/Environments/Landscape/SmallFoliage/MasterMaterial/MaterialFunctions/MF_BasicGrass}.Result) {materialFunction=/Game/Athena/Apollo/Environments/Landscape/SmallFoliage/MasterMaterial/MaterialFunctions/MF_LGRock}.Result) {materialFunction=/Game/Athena/Apollo/Environments/Landscape/SmallFoliage/MasterMaterial/MaterialFunctions/MF_Leaves}
  n4 = BreakMaterialAttributes!(Attr=MaterialFunctionCall!(
    IsStaticSnow=StaticBool()
    UseAlternateStaticSnowMask
    AlternateStaticSnowMask
    "UseRoad&&OcclusionMask"
    IsForGrassMeshToPushDown=@n1
    AdditionalMaskToRemoveSnow
    IsFallingSnowFX
    UseMinimapCoords
    IncreaseSnowUsingSlope=StaticBool()
    UseSmoothBorder
    LerpHighM
    TemporalHeightLerpContrast
    IsLobbyIsland=StaticBoolParameter() {parameterName=IsForLobby}
    UseBR-IslandMask
    IsTree
    IsForRenderTarget
    UseSparkle
    UseSnowNormals
    UseAdditionalMaskToRemoveSnow
    IsTangentSpaceInput
    UseVertexNormalsForBasicSnow
    UseTangentSnowNormals=StaticBool()
    Material=MaterialFunctionCall!(In=@n3.Result) {materialFunction=/Game/Athena/Apollo/Environments/Landscape/SmallFoliage/MasterMaterial/MaterialFunctions/MF_JunglePlants}.Result
   ) {materialFunction=/Game/Athena/Apollo/Environments/Landscape/Materials/Biome/UniqueBiomeMaterialFunctions/MF_Apollo_Snow_Progression}.Material)
  n5 = Multiply(A=@n4.BaseColor, B=ScalarParameter() {defaultValue=1.35, parameterName=GrassBrightenMultiplier})
  n6 = ScalarParameter() {defaultValue=0.005, parameterName=GrassSaturation}
  n7 = MaterialFunctionCall!(
   In=MakeMaterialAttributes!(
    BaseColor=Desaturation(None=@n5, Fraction=@n6)
    Metallic=@n4.Metallic
    Specular=@n4.Metallic
    Roughness=@n4.Metallic
    Anisotropy=@n4.Metallic
    EmissiveColor=@n4.Metallic
    Opacity=@n4.Metallic
    OpacityMask=@n4.Metallic
    Normal=@n4.Metallic
    Tangent=@n4.Metallic
    WorldPositionOffset=@n4.Metallic
    SubsurfaceColor=@n4.Metallic
    ClearCoat=@n4.Metallic
    ClearCoatRoughness=@n4.Metallic
    AmbientOcclusion=@n4.Metallic
    Refraction=@n4.Metallic
    CustomizedUVs_0
    CustomizedUVs_1
    CustomizedUVs_2
    CustomizedUVs_3
    CustomizedUVs_4
    CustomizedUVs_5
    CustomizedUVs_6
    CustomizedUVs_7
    PixelDepthOffset=@n4.Metallic
    ShadingModel=@n4.Metallic
    Displacement=@n4.Metallic
   )
   UseHeightFalloff=StaticBool()
   OnFireEmissiveOverride
   HeightFalloffRadius
   HeightFalloffHardness
   VariationMasksMobile
   In_Grass
   Is_Stylised_WE
  ) {materialFunction=/Game/Packages/Fortress_SharedMaterials/Base_Material_Functions/MF_Curie_Landscape}
  n8 = VectorParameter() {defaultValue=<rgba 0.3542 0.1568 0.0664 1>}
  n9 = SetMaterialAttributes!(
   MaterialAttributes=@n7.Result
   Base Color=@n8.RGB
   Opacity=Constant() {r=1}
   Subsurface Color=Multiply(A=@n8.RGB, B=GetMaterialAttributes(None=@n7.Result) {attributeGetTypes [#1]}.Subsurface Color)
   Roughness=ScalarParameter() {defaultValue=0.15}
  ) {attributeSetTypes [
    #2
    #3
    #1
    #4
   ]}
  n10 = GetMaterialAttributes(None=@n7.Result) {attributeGetTypes [#4, #5]}
  n11 = ScalarParameter() {defaultValue=0.5, parameterName=Param_1}
  n12 = ScalarParameter() {defaultValue=0.56, parameterName=Param_2}
  n13 = Add(
   A=MaterialFunctionCall!(World Position) {materialFunction=/Engine/Functions/Engine_MaterialFunctions02/Utility/DepthFromWorldPosition}.Result
   B=ScalarParameter() {defaultValue=-1000, parameterName=Param_4}
  )
  n14 = ScalarParameter() {defaultValue=1000, parameterName=Param_3}
  n15 = BlendMaterialAttributes(
   A=SetMaterialAttributes!(
    MaterialAttributes=@n10.MaterialAttributes
    Roughness=Multiply(A=@n10.Roughness, B=@n11)
    Specular=Multiply(A=@n10.Specular, B=@n12) {constB=0.5}
   ) {attributeSetTypes [#4, #5]}
   B=@n7.Result
   Alpha=Saturate(None=Divide(A=@n13, B=@n14))
  )
  n16 = Saturate(None=DistanceFieldApproxAO!(
    World Position
    Normal
    BaseDistance=ScalarParameter() {defaultValue=155, group=C_Dirt_AO, parameterName=Details AO Distance, sortPriority=22}
    Radius=ScalarParameter() {defaultValue=512, group=C_Dirt_AO, parameterName=Details AO Fade, sortPriority=23}
   ) {numSteps=3})
  StaticSwitchParameter(
   True=ShadingPathSwitch(Default=BlendMaterialAttributes(A=@n9, B=@n15, Alpha=@n16), Mobile=@n7.Result) {desc=Migrated from FeatureLevelSwitch}
   False=@n7.Result
  ) {group=Jungle, parameterName=Jungle Test}
  TextureSample() {samplerType=SAMPLERTYPE_Normal, texture=/Game/Environments/Asteria/Terrain/Textures/T_Asteria_Mud_01_N}
  n17 = TextureCoordinate()
  n18 = ScalarParameter() {defaultValue=2, parameterName=Param_5}
  Transform(None=TextureSample(UVs=Multiply(A=@n17, B=@n18)) {samplerType=SAMPLERTYPE_Normal, texture=/Game/Environments/Apollo/Sets/Terrain/Textures/T_Apollo_Dirt2_N}.RGB)
  VectorParameter() {parameterName=Param_1}
 ]
}
